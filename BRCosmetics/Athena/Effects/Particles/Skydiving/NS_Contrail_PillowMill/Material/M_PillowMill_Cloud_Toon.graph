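# Unlit masked toon cloud material for Niagara mesh particles in the PillowMill skydiving contrail.
# Vertex normal against camera vector plus a shadow direction gives stepped shading bands, which drive a blend from the tint color.
# Screen-space halftone dots, anchored to object position and scaled by view depth, break up the bands.
# A rounded Fresnel darkens the rim into a hard outline before the shared emissive adjustment function.
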Material /BRCosmetics/Athena/Effects/Particles/Skydiving/NS_Contrail_PillowMill/Material/M_PillowMill_Cloud_Toon
graph {
 properties {
  bAutomaticallySetUsageInEditor=false
  bUsedWithNiagaraMeshParticles=true
  blendMode=BLEND_Masked
  pixelDepthOffsetMode=PDOM_Legacy
  shadingModel=MSM_Unlit
  usedShadingModels=Unlit
 }
 nodes [
  n1 = Round(None=Fresnel(ExponentIn=ScalarParameter() {defaultValue=1.5, parameterName=ToonOutlineSpread}) {baseReflectFraction=0.01, exponent=1})
  n2 = VectorParameter() {parameterName=Color, defaultValue=<rgba 0.9911 0.0921 0.4004 1>}
  n3 = CameraVectorWS()
  n4 = VectorParameter() {parameterName=ToonShadowDirection, defaultValue=<rgba 0.1 0 0.5 1>}
  n5 = Power(Base=DotProduct(A=VertexNormalWS(), B=Add(A=@n3, B=@n4.RGB)), Exp=ScalarParameter() {defaultValue=1, parameterName=ToonBreakupAmount})
  n6 = ScalarParameter() {defaultValue=2, parameterName=ToonShadowSpread}
  n7 = Divide(A=Floor(None=Multiply(A=@n5, B=@n6) {constB=2}), B=@n6)
  n8 = ConstantBiasScale(None=ScreenPosition().ViewportUV) {bias=-0.5, scale=2}
  n9 = MaterialFunctionCall(Float2=ViewSize()) {materialFunction=&BreakOutFloat2Components}
  n10 = AppendVector(A=Constant() {r=1}, B=Divide(A=@n9.G, B=@n9.G))
  n11 = Multiply(
   A=MaterialFunctionCall(World Position=ObjectPositionWS()) {materialFunction=&TransformToClipSpace}.Clip Space XY
   B=Constant2Vector() {g=-1, r=1}
  )
  n12 = Subtract(A=ConstantBiasScale(None=Multiply(A=@n8, B=@n10)), B=Multiply(A=Multiply(A=@n10, B=@n11)) {constB=0.5})
  n13 = TransformPosition(None=ObjectPositionWS()) {transformSourceType=TRANSFORMPOSSOURCE_World, transformType=TRANSFORMPOSSOURCE_View}
  n14 = Multiply(A=Divide(B=ComponentMask(None=@n13) {b=true}))
  n15 = Frac(None=Multiply(
    A=MaterialFunctionCall(UVs=@n12, Texture Scale=@n14) {materialFunction=&ScaleUVsByCenter}.UVs
    B=ScalarParameter() {defaultValue=2, parameterName=DotTiling}
   ) {constB=36})
  n16 = SmoothStep(Value=Length(None=Subtract(A=@n15) {constB=0.5}) {is=class MaterialExpressionLength}) {constMax=0.5, constMin=0.4}
  n17 = LinearInterpolate(A=@n7, Alpha=@n7) {constB=222}
  n18 = LinearInterpolate(
   A=Multiply(A=@n7, B=LinearInterpolate(A=@n16, Alpha=@n7) {constA=1})
   B=LinearInterpolate(A=Constant3Vector() {constant=<rgba 0.9911 0.0921 0.4004 1>}, Alpha=@n16) {constA=1}
   Alpha=Saturate(None=LinearInterpolate(A=@n17, Alpha=@n7) {constB=0})
  )
  MaterialFunctionCall(Emissive=Multiply(A=OneMinus(None=@n1), B=LinearInterpolate(A=@n2.RGB, Alpha=@n18))) {materialFunction=&MF_FX_EmissiveAdjustments}
 ]
}
