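# Unlit, translucent, two-sided Niagara sprite material for the star shapes in the PillowMill skydiving contrail effect.
# Particle color tints the shape texture's alpha, and a StarSelect parameter blends toward a saturated version with the red channel added.
# That blend feeds an emissive adjustment function.
# Texture and particle alpha are multiplied by a smoothstepped near-camera fade driven by CameraFadeDistance.
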Material /BRCosmetics/Athena/Effects/Particles/Skydiving/NS_Contrail_PillowMill/Material/M_PillowMill_Stars
graph {
 properties {
  bAutomaticallySetUsageInEditor=false
  bUsedWithNiagaraSprites=true
  blendMode=BLEND_Translucent
  pixelDepthOffsetMode=PDOM_Legacy
  shadingModel=MSM_Unlit
  twoSided=true
  usedShadingModels=Unlit
 }
 nodes [
  n1 = TextureSample() {texture=/Game/Effects/Fort_Effects/Textures/T_PillowMill_Shapes}
  n2 = ParticleColor()
  n3 = Multiply(A=@n1.A, B=@n2.RGB)
  n4 = Add(A=@n1.R, B=@n3)
  MaterialFunctionCall(Emissive=LinearInterpolate(A=@n3, B=Saturate(None=@n4), Alpha=ScalarParameter() {parameterName=StarSelect})) {materialFunction=&MF_FX_EmissiveAdjustments}
  n5 = ScalarParameter() {defaultValue=120, parameterName=CameraFadeDistance}
  Multiply(
   A=Multiply(A=@n1.A, B=@n2.A)
   B=SmoothStep(Value=MaterialFunctionCall(Distance=@n5) {materialFunction=&Near_Camera_Fade}.Result) {constMin=0.5}
  )
 ]
}
