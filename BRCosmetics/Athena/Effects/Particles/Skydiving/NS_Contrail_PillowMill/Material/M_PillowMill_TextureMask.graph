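# Unlit, translucent, two-sided Niagara sprite material for the PillowMill skydiving contrail, drawing shape masks tinted by particle color.
# A channel-mask parameter picks one channel of the shapes texture, which multiplies particle RGB into an emissive adjustment function.
# The same mask times particle alpha is multiplied by a smoothstepped near-camera fade with an adjustable distance, so sprites fade out close to the camera.
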Material /BRCosmetics/Athena/Effects/Particles/Skydiving/NS_Contrail_PillowMill/Material/M_PillowMill_TextureMask
graph {
 properties {
  bAutomaticallySetUsageInEditor=false
  bUsedWithNiagaraSprites=true
  blendMode=BLEND_Translucent
  pixelDepthOffsetMode=PDOM_Legacy
  shadingModel=MSM_Unlit
  twoSided=true
  usedShadingModels=Unlit
 }
 nodes [
  n1 = ChannelMaskParameter(None=TextureSampleParameter2D() {parameterName=TextureInput, texture=/Game/Effects/Fort_Effects/Textures/T_PillowMill_Shapes}.RGB) {parameterName=TextureChannelSelect}
  n2 = ParticleColor()
  MaterialFunctionCall(Emissive=Multiply(A=@n1.R, B=@n2.RGB)) {materialFunction=&MF_FX_EmissiveAdjustments}
  n3 = ScalarParameter() {defaultValue=120, parameterName=CameraFadeDistance}
  Multiply(
   A=Multiply(A=@n2.A, B=@n1.R)
   B=SmoothStep(Value=MaterialFunctionCall(Distance=@n3) {materialFunction=&Near_Camera_Fade}.Result) {constMin=0.5}
  )
 ]
}
