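# Niagara skydiving contrail cosmetic (PillowMill) trailing pink-and-black comic clouds, lines and stars from the character's skeletal mesh.
# Each emitter partitions particles across bones or sockets with mirrored offsets, then drifts them by age-blended random spread plus contrail setup's direction velocity.
# Dive value drives line and star spawn rates and shortens line lifetimes; lines run in GPU and CPU variants.
# Clouds mix meshes with dot sprites and switch material parameters between gameplay and front end.
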NiagaraSystem /BRCosmetics/Athena/Effects/Particles/Skydiving/NS_Contrail_PillowMill/NS_Contrail_PillowMill
dependencies {
 usedDataInterfaces [
  &NiagaraDataInterfaceArrayFloat
  &NiagaraDataInterfaceVector2DCurve
  &NiagaraDataInterfaceCurve
  &NiagaraDataInterfaceSkeletalMesh
 ]
 usedDynamicInputs []
 usedModules [
  &SystemState
  /Game/Effects/Niagara/Modules/Contrails/FN_Contrail_Setup
  &FN_Integrator
  &self:SystemSpawnScript.SystemScriptSource.SystemScriptGraph.NiagaraNodeAssignment_0.SetVariables_#1
 ]
}
emitters {
 Comic_Clouds {
  emitterUpdateScript [
   &EmitterState()
   &SpawnBurst_Instantaneous(
    Spawn Count=&RandomRangeInt(Minimum=1, Maximum=3)
    Spawn Probability=0.5
    Use Spawn Probability=true
    UseLoopCountLimit=false
   )
   &SpawnRate(SpawnRate=&RandomRangeFloat(Minimum=4, Maximum=6), Use Spawn Probability=false)
  ]
  particleSpawnScript [
   &PartitionParticles()
   /SetVariables(Particles.MeshIndex=&RandomRangeInt(Minimum=0, Maximum=1)) {isSetParameters=true}
   &InitializeParticle(
    Color=&LerpLinearColors(
     StartColor=<rgba 1 0.2668 0.8504 1>
     EndColor=<rgba 0 0 0 1>
     LerpFactor=&MakeFloatFromInt(Int=&RandomRangeInt(Minimum=0, Maximum=1))
    )
    Lifetime Max=1
    Lifetime Min=0.85
    Lifetime Mode=::NewEnumerator1
    Mesh Scale Mode=::NewEnumerator1
    Mesh Uniform Scale Max=1.1
    Mesh Uniform Scale Min=0.5
    UsePositionOffset=false
    Write Material Random=true
   )
   &InitialMeshOrientation(Use Rotation Vector=false)
  ]
  particleUpdateScript [
   &ParticleState()
   &UpdateMeshOrientation(Rotation Vector=&RandomRangeVector(Minimum=<0 0 0>, Maximum=<1 1 1>))
   &SkeletalMeshLocation(
    Bone / Socket Index=$Particles.PartitionParticles.Partition
    Bone Sampling Mode=::NewEnumerator4
    MaskSpawnGroup=false
    Orientation Sampling=::NewEnumerator1
    Sampled Position Offset=&Multiply_VectorByFloat(
     Vector=&Lerp_Vector(A=<-5 0 0>, B=<25 0 0>, Alpha=&RandomRangeFloat(Minimum=0, Maximum=1))
     Float=&SelectFloatFromArray(Direct Array Index=$Particles.PartitionParticles.Partition, Float Selection Array=DataInterface() {propertyValues {FloatData [1,-1]}})
    )
    Skeletal Mesh=$System.Skeletal Mesh
   ) {module=SkeletalMeshLocation001}
   /SetVariables(Particles.NewAge:NiagaraFloat=&Multiply_Float(A=$Particles.Age, B=&RandomRangeFloat(Minimum=0.2, Maximum=0.2))) {isSetParameters=true, module=SetVariables_#2}
   /SetVariables(Particles.Velocity=&Multiply_VectorByFloat(
     Vector=&Lerp_Vector(
      A=<0 0 0>
      B=&Add_Vector(
       A=&RandomRangeVector(Minimum=<-200 -200 0>, Maximum=<200 200 0>)
       B=$System.FN_Contrail_Setup.Direction Velocity
      )
      Alpha=$Particles.NewAge
     )
     Float=1
    )) {isSetParameters=true, module=SetVariables_#3}
   /SetVariables(Particles.Position=&AddVectorToPosition(
     Position=$Particles.Position
     Vector=&Multiply_VectorByFloat(Vector=$Particles.Velocity, Float=$Particles.Age)
    )) {isSetParameters=true, module=SetVariables_#4}
   /SetVariables(
    Particles.DotCameraOffset=-6
    Particles.DotPivotOffset:Vector2f=&RandomRangeVector2D(Minimum=<0.6 0.5>, Maximum=<0.7 0.5>)
    Particles.DotSize:Vector2f=&Vector2DFromFloat(Value=&FloatFromCurve(
      FloatCurve=DataInterface() {propertyValues {Curve {keys [
          {arriveTangent=1.109, interpMode=RCIM_Cubic, leaveTangent=0.2196, tangentMode=RCTM_Break, time=0, value=0}
          {arriveTangent=6.652, interpMode=RCIM_Cubic, leaveTangent=-0.1317, tangentMode=RCTM_Break, time=0.4, value=1}
          {arriveTangent=0.1399, interpMode=RCIM_Cubic, leaveTangent=0.1399, tangentMode=RCTM_User, time=0.9, value=1}
          {arriveTangent=0.0224, interpMode=RCIM_Cubic, leaveTangent=0.0224, tangentMode=RCTM_User, time=1, value=0}
         ]}}}
      Scale Curve=&RandomRangeFloat(Minimum=35, Maximum=45)
     ))
    Particles.DotsColor:LinearColor=&MakeLinearColorFromVectorAndFloat(
     "Vector (RGB)"=&MakeVectorFromLinearColorRGB(Color=&RandomRangeLinearColor(Minimum=<rgba 0.9911 0.1998 0.4712 1>, Maximum=<rgba 0.9911 0.2644 0.5136 1>))
     "Float (Alpha)"=&UniformAOrBFloat(
      A=&MakeFloatFromInt(Int=0)
      B=&MakeFloatFromInt(Int=1)
      Evaluation Type=::NewEnumerator0
      Recalculate Random Each Loop=true
      Override Seed=false
     )
    )
   ) {isSetParameters=true}
   &ScaleMeshSize(Scale Factor=&VectorFromFloat(Value=&FloatFromCurve(FloatCurve=DataInterface() {propertyValues {Curve {keys [
          {arriveTangent=4.1785, interpMode=RCIM_Cubic, leaveTangent=-0.0499, tangentMode=RCTM_Break, time=0, value=0}
          {arriveTangent=6.9884, interpMode=RCIM_Cubic, leaveTangent=0.2196, tangentMode=RCTM_Break, time=0.3, value=1}
          <at 0.6 1.3 RCIM_Cubic>
          {arriveTangent=0.1399, interpMode=RCIM_Cubic, leaveTangent=0.1399, tangentMode=RCTM_User, time=0.9, value=1}
          {arriveTangent=0.0224, interpMode=RCIM_Cubic, leaveTangent=0.0224, tangentMode=RCTM_User, time=1, value=0}
         ]}}})))
   &DynamicMaterialParameters(
    Index 0 Param 1=1
    Index 0 Param 2=&FN_FrontEndFloatModifier(Gameplay=0, Front End=0)
    Index 0 Param 3=&FloatFromCurve(
     FloatCurve=DataInterface() {propertyValues {Curve {keys [
         {interpMode=RCIM_Cubic, tangentMode=RCTM_User, time=0.65, value=0}
         {interpMode=RCIM_Cubic, tangentMode=RCTM_User, time=0.8, value=0.9}
         {arriveTangent=-4.5, interpMode=RCIM_Cubic, leaveTangent=-4.5, time=0.9, value=1}
         {interpMode=RCIM_Cubic, tangentMode=RCTM_User, time=1, value=0}
        ]}}}
     Scale Curve=2
    )
    Index 0 Param 4=&FloatFromCurve(
     FloatCurve=DataInterface() {propertyValues {Curve {keys [
         {interpMode=RCIM_Cubic, tangentMode=RCTM_User, time=0.05, value=1}
         {interpMode=RCIM_Cubic, tangentMode=RCTM_User, time=0.4, value=0}
         <at 0.8 0.125>
         <at 1 0.125>
        ]}}}
     Scale Curve=-10
    )
    Index 3 Param 1=&FN_FrontEndFloatModifier(Gameplay=3.5146, Front End=-30)
    Index 3 Param 2=&FN_FrontEndFloatModifier(Gameplay=35.2929, Front End=2)
    Index 3 Param 3=&FN_FrontEndFloatModifier(Gameplay=121.0378, Front End=1.4)
    Module.Param0WriteEnabled=true
    Module.Param0WriteEnabled3=true
    Module.Param1WriteEnabled=true
    Module.Param1WriteEnabled3=true
    Module.Param2WriteEnabled=true
    Module.Param2WriteEnabled3=true
    Module.Param3WriteEnabled=true
    Module.Param3WriteEnabled3=false
    Write Parameter Index 3=true
   )
   &CameraOffset(Camera Offset Amount=&RandomRangeFloat(Minimum=1, Maximum=-5))
   &DynamicMaterialParameters(
    Index 0 Param 1=&RandomRangeFloat(Minimum=8, Maximum=32)
    Module.Index0Param0Write=true
    Module.Index0Param1Write=false
    Module.Index0Param2Write=false
    Module.Index0Param3Write=false
   ) {module=DynamicMaterialParameters001}
  ]
  rendererClasses [&NiagaraSpriteRendererProperties, &NiagaraMeshRendererProperties]
  renderers [
   {rendererClass=&NiagaraSpriteRendererProperties}
   {rendererClass=&NiagaraMeshRendererProperties, rendererIndex=1}
  ]
 }
 Comic_Lines {
  simTarget=GPUComputeSim
  emitterUpdateScript [
   &EmitterState()
   &SpawnRate(
    SpawnRate=&Lerp_Float(A=&RandomRangeFloat(Minimum=30, Maximum=35), B=200, Alpha=$System.FN_Contrail_Setup.Dive Value)
    Use Spawn Probability=false
   ) {module=SpawnRate001}
  ]
  particleSpawnScript [
   &PartitionParticles()
   &InitializeParticle(
    Color=&LerpLinearColors(
     StartColor=<rgba 0.9911 0.1998 0.4712 1>
     EndColor=<rgba 0 0 0 1>
     LerpFactor=&UniformAOrBFloat(
      A=&MakeFloatFromInt(Int=0)
      B=&MakeFloatFromInt(Int=1)
      Distribution Weight=0.3
      Evaluation Type=::NewEnumerator0
      Recalculate Random Each Loop=true
      Override Seed=false
     )
    )
    Lifetime=&Lerp_Float(
     A=&RandomRangeFloat(Minimum=0.4, Maximum=0.6)
     B=&RandomRangeFloat(Minimum=0.1, Maximum=0.2)
     Alpha=$System.FN_Contrail_Setup.Dive Value
    )
    Sprite Size Max=<30 88>
    Sprite Size Min=<20 40>
    Sprite Size Mode=::NewEnumerator2
    UsePositionOffset=false
    Write Material Random=true
   ) {module=InitializeParticle001}
   /SetVariables(Particles.UVRandom:Vector2f=&MakeVector2D(
     X=&UniformAOrBFloat(A=-1, Evaluation Type=::NewEnumerator0, Recalculate Random Each Loop=true, Override Seed=false)
     Y=1
    )) {isSetParameters=true, module=SetVariables_#5}
  ]
  particleUpdateScript [
   &ParticleState()
   &SkeletalMeshLocation(
    Bone / Socket Index=$Particles.PartitionParticles.Partition
    Bone Sampling Mode=::NewEnumerator4
    MaskSpawnGroup=false
    Orientation Sampling=::NewEnumerator1
    Sampled Position Offset=&Multiply_VectorByFloat(
     Vector=&Lerp_Vector(A=<-5 0 0>, B=<25 0 0>, Alpha=&RandomRangeFloat(Minimum=0, Maximum=1))
     Float=&SelectFloatFromArray(Direct Array Index=$Particles.PartitionParticles.Partition, Float Selection Array=DataInterface() {propertyValues {FloatData [1,-1]}})
    )
    Skeletal Mesh=$System.Skeletal Mesh
   ) {module=SkeletalMeshLocation001}
   /SetVariables(Particles.NewAge:NiagaraFloat=&Multiply_Float(A=$Particles.Age, B=&Lerp_Float(A=0.75, B=4, Alpha=$System.FN_Contrail_Setup.Dive Value))) {isSetParameters=true, module=SetVariables_#2}
   /SetVariables(Particles.Velocity=&Multiply_VectorByFloat(
     Vector=&Lerp_Vector(
      A=<0 0 0>
      B=&Add_Vector(
       A=&RandomRangeVector(Minimum=<-120 99 0>, Maximum=<120 -99 0>)
       B=$System.FN_Contrail_Setup.Direction Velocity
      )
      Alpha=$Particles.NewAge
     )
     Float=1
    )) {isSetParameters=true, module=SetVariables_#3}
   /SetVariables(Particles.Position=&AddVectorToPosition(
     Position=$Particles.Position
     Vector=&Multiply_VectorByFloat(Vector=$Particles.Velocity, Float=$Particles.Age)
    )) {isSetParameters=true, module=SetVariables_#4}
   /SetVariables(
    Particles.DotCameraOffset=-6
    Particles.DotsColor:LinearColor=&MakeLinearColorFromVectorAndFloat(
     "Vector (RGB)"=&MakeVectorFromLinearColorRGB(Color=&RandomRangeLinearColor(Minimum=<rgba 0.9911 0.1998 0.4712 1>, Maximum=<rgba 0.9911 0.2644 0.5136 1>))
     "Float (Alpha)"=&UniformAOrBFloat(
      A=&MakeFloatFromInt(Int=0)
      B=&MakeFloatFromInt(Int=1)
      Distribution Weight=0.4
      Evaluation Type=::NewEnumerator0
      Recalculate Random Each Loop=true
      Override Seed=false
     )
    )
    Particles.DotsSize:Vector2f=&Vector2DFromCurve(
     Vector2Curve=DataInterface() {propertyValues {
       XCurve {keys [{tangentMode=RCTM_User, time=0, value=1}, {tangentMode=RCTM_User, time=0.9, value=1}]}
       YCurve {keys [
         {arriveTangent=-0.153, interpMode=RCIM_Cubic, leaveTangent=0.2381, tangentMode=RCTM_Break, time=0, value=0}
         {arriveTangent=-0.0772, interpMode=RCIM_Cubic, leaveTangent=-0.0772, tangentMode=RCTM_User, time=0.3, value=1.025}
         {arriveTangent=-0.0037, interpMode=RCIM_Cubic, leaveTangent=0.1196, tangentMode=RCTM_Break, time=0.9, value=1}
         {arriveTangent=0.0583, interpMode=RCIM_Cubic, tangentMode=RCTM_Break, time=1, value=0}
        ]}
      }}
     ScaleCurve=&RandomRangeVector2D(Minimum=<90 80>, Maximum=<120 80>)
    )
   ) {isSetParameters=true, module=SetVariables_#6}
   &ScaleSpriteSize(
    Non-Uniform Curve Sprite Scale=DataInterface() {propertyValues {
      XCurve {keys [{tangentMode=RCTM_User, time=0, value=1}, {tangentMode=RCTM_User, time=1, value=1}]}
      YCurve {keys [
        {interpMode=RCIM_Cubic, leaveTangent=0.5585, tangentMode=RCTM_Break, time=0, value=0}
        {arriveTangent=6.8019, interpMode=RCIM_Cubic, leaveTangent=0.1196, tangentMode=RCTM_Break, time=0.3, value=1}
        {arriveTangent=-0.153, interpMode=RCIM_Cubic, leaveTangent=-1.4286, tangentMode=RCTM_Break, time=0.9, value=1}
        {arriveTangent=0.0583, interpMode=RCIM_Cubic, tangentMode=RCTM_Break, time=1, value=0}
       ]}
     }}
    Scale Sprite Size Mode=::NewEnumerator3
   )
   &CameraOffset(Camera Offset Amount=&RandomRangeFloat(Minimum=1, Maximum=-5))
   &DynamicMaterialParameters(
    Index 0 Param 1=&RandomRangeFloat(Minimum=32, Maximum=64)
    Module.Index0Param0Write=true
    Module.Index0Param1Write=true
    Module.Index0Param2Write=true
    Module.Index0Param3Write=true
   )
  ]
  rendererClasses [&NiagaraSpriteRendererProperties]
  renderers [
   {rendererClass=&NiagaraSpriteRendererProperties}
   {rendererClass=&NiagaraSpriteRendererProperties, rendererIndex=1}
  ]
 }
 Comic_Lines_CPU {
  emitterUpdateScript [
   &EmitterState()
   &SpawnRate(
    SpawnRate=&Lerp_Float(A=&RandomRangeFloat(Minimum=15, Maximum=20), B=20, Alpha=$System.FN_Contrail_Setup.Dive Value)
    Use Spawn Probability=false
   ) {module=SpawnRate001}
  ]
  particleSpawnScript [
   &PartitionParticles()
   &InitializeParticle(
    Color=&LerpLinearColors(
     StartColor=<rgba 0.9911 0.1998 0.4712 1>
     EndColor=<rgba 0 0 0 1>
     LerpFactor=&UniformAOrBFloat(
      A=&MakeFloatFromInt(Int=0)
      B=&MakeFloatFromInt(Int=1)
      Distribution Weight=0.3
      Evaluation Type=::NewEnumerator0
      Recalculate Random Each Loop=true
      Override Seed=false
     )
    )
    Lifetime=&Lerp_Float(
     A=&RandomRangeFloat(Minimum=0.4, Maximum=0.6)
     B=&RandomRangeFloat(Minimum=0.1, Maximum=0.2)
     Alpha=$System.FN_Contrail_Setup.Dive Value
    )
    Sprite Size Max=<30 88>
    Sprite Size Min=<20 40>
    Sprite Size Mode=::NewEnumerator2
    UsePositionOffset=false
    Write Material Random=true
   ) {enabled=false, module=InitializeParticle001}
   &InitializeParticle(
    Color=&LerpLinearColors(
     StartColor=<rgba 0.9911 0.1998 0.4712 1>
     EndColor=<rgba 0 0 0 1>
     LerpFactor=&UniformAOrBFloat(
      A=&MakeFloatFromInt(Int=0)
      B=&MakeFloatFromInt(Int=1)
      Distribution Weight=0.3
      Evaluation Type=::NewEnumerator0
      Recalculate Random Each Loop=true
      Override Seed=false
     )
    )
    Lifetime=&Lerp_Float(
     A=&RandomRangeFloat(Minimum=0.4, Maximum=0.6)
     B=&RandomRangeFloat(Minimum=0.1, Maximum=0.2)
     Alpha=$System.FN_Contrail_Setup.Dive Value
    )
    Sprite Size Max=<30 88>
    Sprite Size Min=<20 40>
    Sprite Size Mode=::NewEnumerator2
    UsePositionOffset=false
    Write Material Random=true
   )
   /SetVariables(Particles.UVRandom:Vector2f=&MakeVector2D(
     X=&UniformAOrBFloat(A=-1, Evaluation Type=::NewEnumerator0, Recalculate Random Each Loop=true, Override Seed=false)
     Y=1
    )) {isSetParameters=true, module=SetVariables_#5}
  ]
  particleUpdateScript [
   &ParticleState()
   &SkeletalMeshLocation(
    Bone / Socket Index=$Particles.PartitionParticles.Partition
    Bone Sampling Mode=::NewEnumerator4
    MaskSpawnGroup=false
    Orientation Sampling=::NewEnumerator1
    Sampled Position Offset=&Multiply_VectorByFloat(
     Vector=&Lerp_Vector(A=<-5 0 0>, B=<25 0 0>, Alpha=&RandomRangeFloat(Minimum=0, Maximum=1))
     Float=&SelectFloatFromArray(Direct Array Index=$Particles.PartitionParticles.Partition, Float Selection Array=DataInterface() {propertyValues {FloatData [1,-1]}})
    )
    Skeletal Mesh=$System.Skeletal Mesh
   ) {module=SkeletalMeshLocation001}
   /SetVariables(Particles.NewAge:NiagaraFloat=&Multiply_Float(A=$Particles.Age, B=&Lerp_Float(A=0.75, B=4, Alpha=$System.FN_Contrail_Setup.Dive Value))) {isSetParameters=true, module=SetVariables_#2}
   /SetVariables(Particles.Velocity=&Multiply_VectorByFloat(
     Vector=&Lerp_Vector(
      A=<0 0 0>
      B=&Add_Vector(
       A=&RandomRangeVector(Minimum=<-120 99 0>, Maximum=<120 -99 0>)
       B=$System.FN_Contrail_Setup.Direction Velocity
      )
      Alpha=$Particles.NewAge
     )
     Float=1
    )) {isSetParameters=true, module=SetVariables_#3}
   /SetVariables(Particles.Position=&AddVectorToPosition(
     Position=$Particles.Position
     Vector=&Multiply_VectorByFloat(Vector=$Particles.Velocity, Float=$Particles.Age)
    )) {isSetParameters=true, module=SetVariables_#4}
   /SetVariables(
    Particles.DotCameraOffset=-6
    Particles.DotsColor:LinearColor=&MakeLinearColorFromVectorAndFloat(
     "Vector (RGB)"=&MakeVectorFromLinearColorRGB(Color=&RandomRangeLinearColor(Minimum=<rgba 0.9911 0.1998 0.4712 1>, Maximum=<rgba 0.9911 0.2644 0.5136 1>))
     "Float (Alpha)"=&UniformAOrBFloat(
      A=&MakeFloatFromInt(Int=0)
      B=&MakeFloatFromInt(Int=1)
      Distribution Weight=0.4
      Evaluation Type=::NewEnumerator0
      Recalculate Random Each Loop=true
      Override Seed=false
     )
    )
    Particles.DotsSize:Vector2f=&Vector2DFromCurve(
     Vector2Curve=DataInterface() {propertyValues {
       XCurve {keys [{tangentMode=RCTM_User, time=0, value=1}, {tangentMode=RCTM_User, time=0.9, value=1}]}
       YCurve {keys [
         {arriveTangent=-0.153, interpMode=RCIM_Cubic, leaveTangent=0.2381, tangentMode=RCTM_Break, time=0, value=0}
         {arriveTangent=-0.0772, interpMode=RCIM_Cubic, leaveTangent=-0.0772, tangentMode=RCTM_User, time=0.3, value=1.025}
         {arriveTangent=-0.0037, interpMode=RCIM_Cubic, leaveTangent=0.1196, tangentMode=RCTM_Break, time=0.9, value=1}
         {arriveTangent=0.0583, interpMode=RCIM_Cubic, tangentMode=RCTM_Break, time=1, value=0}
        ]}
      }}
     ScaleCurve=&RandomRangeVector2D(Minimum=<90 80>, Maximum=<120 80>)
    )
   ) {isSetParameters=true, module=SetVariables_#6}
   &ScaleSpriteSize(
    Non-Uniform Curve Sprite Scale=DataInterface() {propertyValues {
      XCurve {keys [{tangentMode=RCTM_User, time=0, value=1}, {tangentMode=RCTM_User, time=1, value=1}]}
      YCurve {keys [
        {interpMode=RCIM_Cubic, leaveTangent=0.5585, tangentMode=RCTM_Break, time=0, value=0}
        {arriveTangent=6.8019, interpMode=RCIM_Cubic, leaveTangent=0.1196, tangentMode=RCTM_Break, time=0.3, value=1}
        {arriveTangent=-0.153, interpMode=RCIM_Cubic, leaveTangent=-1.4286, tangentMode=RCTM_Break, time=0.9, value=1}
        {arriveTangent=0.0583, interpMode=RCIM_Cubic, tangentMode=RCTM_Break, time=1, value=0}
       ]}
     }}
    Scale Sprite Size Mode=::NewEnumerator3
   )
   &CameraOffset(Camera Offset Amount=&RandomRangeFloat(Minimum=1, Maximum=-5))
   &DynamicMaterialParameters(
    Index 0 Param 1=&RandomRangeFloat(Minimum=32, Maximum=64)
    Module.Index0Param0Write=true
    Module.Index0Param1Write=true
    Module.Index0Param2Write=true
    Module.Index0Param3Write=true
   )
  ]
  rendererClasses [&NiagaraSpriteRendererProperties]
  renderers [
   {rendererClass=&NiagaraSpriteRendererProperties}
   {rendererClass=&NiagaraSpriteRendererProperties, rendererIndex=1}
  ]
 }
 Comic_Stars {
  emitterUpdateScript [
   &EmitterState()
   &SpawnRate(SpawnRate=&Lerp_Float(A=11, B=20, Alpha=$System.FN_Contrail_Setup.Dive Value), Use Spawn Probability=false)
  ]
  particleSpawnScript [
   &PartitionParticles()
   &InitializeParticle(
    Color=&LerpLinearColors(
     StartColor=<rgba 0.9911 0.1998 0.4712 1>
     EndColor=<rgba 0 0 0 1>
     LerpFactor=&MakeFloatFromInt(Int=&RandomRangeInt(Minimum=0, Maximum=1))
    )
    Lifetime Max=0.7
    Lifetime Min=0.5
    Lifetime Mode=::NewEnumerator1
    Sprite Size=&RandomRangeVector2D(Minimum=<10 10>, Maximum=<12 18>)
    Sprite Size Mode=::NewEnumerator4
    UsePositionOffset=false
    Write Material Random=true
   )
   /SetVariables(Particles.VisibilityTag=&RandomRangeInt(Minimum=0, Maximum=1)) {isSetParameters=true}
  ]
  particleUpdateScript [
   &ParticleState()
   &SkeletalMeshLocation(
    Bone / Socket Index=$Particles.PartitionParticles.Partition
    Bone Sampling Mode=::NewEnumerator4
    MaskSpawnGroup=false
    Orientation Sampling=::NewEnumerator1
    Sampled Position Offset=&Multiply_VectorByFloat(
     Vector=&Lerp_Vector(A=<-5 0 0>, B=<25 0 0>, Alpha=&RandomRangeFloat(Minimum=0, Maximum=1))
     Float=&SelectFloatFromArray(Direct Array Index=$Particles.PartitionParticles.Partition, Float Selection Array=DataInterface() {propertyValues {FloatData [1,-1]}})
    )
    Skeletal Mesh=$System.Skeletal Mesh
   ) {module=SkeletalMeshLocation001}
   /SetVariables(Particles.NewAge:NiagaraFloat=&Multiply_Float(A=$Particles.Age, B=&Lerp_Float(A=0.5, B=0.6, Alpha=$System.FN_Contrail_Setup.Dive Value))) {isSetParameters=true, module=SetVariables_#2}
   /SetVariables(Particles.Velocity=&Multiply_VectorByFloat(
     Vector=&Lerp_Vector(
      A=<0 0 0>
      B=&Add_Vector(
       A=&RandomRangeVector(Minimum=<-400 -99 0>, Maximum=<400 99 0>)
       B=$System.FN_Contrail_Setup.Direction Velocity
      )
      Alpha=$Particles.NewAge
     )
     Float=1
    )) {isSetParameters=true, module=SetVariables_#3}
   /SetVariables(Particles.Position=&AddVectorToPosition(
     Position=$Particles.Position
     Vector=&Multiply_VectorByFloat(Vector=$Particles.Velocity, Float=$Particles.Age)
    )) {isSetParameters=true, module=SetVariables_#4}
   /SetVariables(
    Particles.DotCameraOffset=-6
    Particles.DotPivotOffset:Vector2f=&RandomRangeVector2D(Minimum=<0.6 1>, Maximum=<0.7 1>)
    Particles.DotSize:Vector2f=&Vector2DFromFloat(Value=&FloatFromCurve(
      FloatCurve=DataInterface() {propertyValues {Curve {keys [
          {arriveTangent=1.109, interpMode=RCIM_Cubic, leaveTangent=0.2196, tangentMode=RCTM_Break, time=0, value=1}
          {arriveTangent=0.1399, interpMode=RCIM_Cubic, leaveTangent=0.1399, tangentMode=RCTM_User, time=0.9, value=1}
          {arriveTangent=0.0224, interpMode=RCIM_Cubic, leaveTangent=0.0224, tangentMode=RCTM_User, time=1, value=0}
         ]}}}
      Scale Curve=&RandomRangeFloat(Minimum=20, Maximum=35)
     ))
    Particles.DotsColor:LinearColor=&MakeLinearColorFromVectorAndFloat(
     "Vector (RGB)"=&MakeVectorFromLinearColorRGB(Color=&RandomRangeLinearColor(Minimum=<rgba 0.9911 0.1998 0.4712 1>, Maximum=<rgba 0.9911 0.2644 0.5136 1>))
     "Float (Alpha)"=&MakeFloatFromInt(Int=&RandomRangeInt(Minimum=0, Maximum=1))
    )
   ) {isSetParameters=true}
   /SetVariables(Particles.StarOutlineSize:Vector2f=&Vector2DFromCurve(
     Vector2Curve=DataInterface() {propertyValues {
       XCurve {keys [
         {interpMode=RCIM_Cubic, tangentMode=RCTM_Break, time=0, value=0}
         {interpMode=RCIM_Cubic, tangentMode=RCTM_Break, time=0.2, value=1}
         {interpMode=RCIM_Cubic, tangentMode=RCTM_Break, time=0.9, value=1}
         {interpMode=RCIM_Cubic, tangentMode=RCTM_Break, time=1, value=0}
        ]}
       YCurve {keys [
         {interpMode=RCIM_Cubic, tangentMode=RCTM_Break, time=0, value=0}
         {arriveTangent=1.1111, interpMode=RCIM_Cubic, leaveTangent=-0.0528, tangentMode=RCTM_Break, time=0.2, value=1}
         {interpMode=RCIM_Cubic, tangentMode=RCTM_Break, time=0.9, value=1}
         {interpMode=RCIM_Cubic, tangentMode=RCTM_Break, time=1, value=0}
        ]}
      }}
     ScaleCurve=&RandomRangeVector2D(Minimum=<20 20>, Maximum=<22 40>)
    )) {isSetParameters=true}
   &ScaleSpriteSize(Uniform Curve Sprite Scale=DataInterface() {propertyValues {Curve {keys [
        {arriveTangent=4.1785, interpMode=RCIM_Cubic, leaveTangent=-0.0499, tangentMode=RCTM_Break, time=0, value=0}
        {arriveTangent=1.109, interpMode=RCIM_Cubic, leaveTangent=0.2196, tangentMode=RCTM_Break, time=0.15, value=1}
        {arriveTangent=0.1399, interpMode=RCIM_Cubic, leaveTangent=0.1399, tangentMode=RCTM_User, time=0.9, value=1}
        {arriveTangent=0.0224, interpMode=RCIM_Cubic, leaveTangent=0.0224, tangentMode=RCTM_User, time=1, value=0}
       ]}}})
  ]
  rendererClasses [&NiagaraSpriteRendererProperties]
  renderers [
   {rendererClass=&NiagaraSpriteRendererProperties}
   {rendererClass=&NiagaraSpriteRendererProperties, rendererIndex=1}
   {rendererClass=&NiagaraSpriteRendererProperties, rendererIndex=2}
  ]
 }
}
summary {
 systemName=NS_Contrail_PillowMill
 emitters [
  {emitterName=Comic_Lines, simTarget=GPUComputeSim, rendererClasses [&NiagaraSpriteRendererProperties]}
  {emitterName=Comic_Stars, rendererClasses [&NiagaraSpriteRendererProperties]}
  {emitterName=Comic_Clouds, rendererClasses [&NiagaraSpriteRendererProperties, &NiagaraMeshRendererProperties]}
  {emitterName=Comic_Lines_CPU, rendererClasses [&NiagaraSpriteRendererProperties]}
 ]
 userVariables [
  User.bIsFrontEnd {type=NiagaraBool}
  User.bIsFrontEndPreview {type=NiagaraBool, defaultValue=true}
  User.ParaGlideLeanAlpha {type=NiagaraFloat, defaultValue=0, description=""}
 ]
}
